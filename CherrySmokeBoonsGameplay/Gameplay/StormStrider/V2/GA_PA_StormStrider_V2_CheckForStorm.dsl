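; Passive Storm Strider V2 ability that feeds energy regen to the player during tactical sprint, at a rate set by storm exposure.
; On activation it caches the pawn's energy component, then uses gameplay tag wait tasks to track safe zone, sprint and sliding.
; Safe zone changes pick the default or in-storm scalable rate and re-evaluate regen against sprint state.
; Helper events start or stop the tagged energy source as sprint begins, ends or sliding starts.
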(event Ability|EventFailedToActivatePassiveAbility
  (params (ActorInfo :type Gameplay%20Ability%20Actor%20Info%20Structure))
  (|CallParentFunction ActorInfo))

(event Ability|EventActivateAbility
  (|CallParentFunction)
  (Utilities|FlowControl|Sequence
    (:then_0
      ;[0.1165 0.4432 1] Store the energy component
      (bind _outfortplayerpawn (CallFunction|GetAvatarasValidFortPlayerPawn)
        (:ValidFortPlayerPawn
          (Utilities|IsValid (Actor|GetComponentbyClass _outfortplayerpawn "/Script/EnergyRuntime.FortComponent_Energy")
            (:Is%20Valid
              (Variables|Default|SetEnergyComponent (Actor|GetComponentbyClass _outfortplayerpawn "/Script/EnergyRuntime.FortComponent_Energy")))))))
    (:then_1
      ;[0.6977 0 1] Adjust the regen value based on being in the storm or not.
      (Ability|Tasks|WaitGameplayTagRemove "(TagName=\"Gameplay.InsideSafeZone\")"
        (:then
          (Ability|Tasks|WaitGameplayTagAdd "(TagName=\"Gameplay.InsideSafeZone\")"
            (:Added
              (Variables|Default|SetEnergyRegenValue (ScalableFloat|GetValueatLevel (Variables|Default|GetSF_DefaultEnergyRegen))))))
        (:Removed
          (Variables|Default|SetEnergyRegenValue (ScalableFloat|GetValueatLevel (Variables|Default|GetSF_InStormEnergyRegen)))))
      ;[1 0.4459 0] If sprinting when entering/exiting the storm, update the applied regen
      (bind _outfortplayerpawn_2 (CallFunction|GetAvatarasValidFortPlayerPawn)
        (:ValidFortPlayerPawn
          (if (GameplayTags|HasMatchingGameplayTag _outfortplayerpawn_2 "(TagName=\"Gameplay.Action.Player.TacticalSprint\")")
            (CallFunction|StopEnergyRegen)
            (CallFunction|StartEnergyRegen)))))
    (:then_2
      ;[0.1686 1 0.7886] Apply regen while sprinting
      (Ability|Tasks|WaitGameplayTagAdd "(TagName=\"Gameplay.Action.Player.TacticalSprint\")"
        (:then
          (Ability|Tasks|WaitGameplayTagRemove "(TagName=\"Gameplay.Action.Player.TacticalSprint\")"
            (:then
              ;[1 0 0.6013] Stop regen if slding
              (Ability|Tasks|WaitGameplayTagAdd "(TagName=\"MovementMode.Sliding\")"
                (:Added
                  (CallFunction|StopEnergyRegen))))
            (:Removed
              (CallFunction|StopEnergyRegen))))
        (:Added
          (CallFunction|StartEnergyRegen))))))

(event Custom|StartEnergyRegen
  (bind _energycomponent (Variables|Default|GetEnergyComponent)
    (:then
      (Energy|StartAddingEnergy _energycomponent (Variables|Default|GetEnergyRegenValue) (Variables|Default|GetT_Source) 0))
    (:else)))

(event Custom|StopEnergyRegen
  (bind _energycomponent (Variables|Default|GetEnergyComponent)
    (:then
      (Energy|StopAddingEnergy _energycomponent (Variables|Default|GetT_Source) 0))
    (:else)))
